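# Animation Blueprint for the female medium BerryFlight player body cosmetic, adding secondary motion to a pose copied from the attached parent mesh.
# A Control Rig takes parachute and front-end flags, then weighted constraints position bag and ring base bones.
# Glute bones get an offset plus layered angular and linear spring AnimDynamics, used only when bPbBlendEnabled is true.
# Spring AnimDynamics on the foot and clavicle rings always finishes the pose.
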AnimBlueprint /BRCosmetics/Characters/Player/Female/Medium/Bodies/F_MED_BerryFlight/Meshes/F_MED_BerryFlight_AnimBP
target_skeleton=&/F_MED_BerryFlight_Skeleton
graphs [
 AnimGraph {nodes [
   n1 = Use cached pose 'PreDyn'()
   n2 = Use cached pose 'GluteDynamics_1'()
   n3 = Get bPbBlendEnabled() -> (bPbBlendEnabled:bool)
   n4 = Anim Dynamics - Bone: foot_ring_bk_l(
    ComponentPose=Local To Component(LocalPose=Blend Poses by bool(BlendPose_0=@n2, BlendPose_1=@n1, bActiveValue=@n3))
    Alpha=1
    ExternalForce="0,0,0"
   ) {
    angularDampingOverride=1
    angularSpringConstant=500
    bAngularSpring=true
    bOverrideAngularDamping=true
    bUseGravityOverride=true
    linearDampingOverride=1
    linearSpringConstant=50
    simulationSpace=RootRelative
    alphaBoolBlend {blendOption=Linear}
    boundBone {boneName=foot_ring_bk_l}
    componentAppliedLinearAccClamp=<3000 3000 3000>
    componentLinearAccScale=<0.4 0.4 0.4>
    componentLinearVelScale=<0.4 0.4 0.4>
    gravityOverride {z=-980}
    physicsBodyDefinitions [{
      boundBone {boneName=foot_ring_bk_l}
      boxExtents=<10 10 10>
      constraintSetup {angularLimitsMax=<0 0 0>, angularLimitsMin=<0 0 -70>, angularTarget=<1 0 0>}
      localJointOffset=<0 3 0>
     }]
   }
   n5 = Anim Dynamics - Bone: foot_ring_bk_r(ComponentPose=@n4, Alpha=1, ExternalForce="0,0,0") {
    angularDampingOverride=1
    angularSpringConstant=500
    bAngularSpring=true
    bOverrideAngularDamping=true
    bUseGravityOverride=true
    linearDampingOverride=1
    linearSpringConstant=50
    simulationSpace=RootRelative
    alphaBoolBlend {blendOption=Linear}
    boundBone {boneName=foot_ring_bk_r}
    componentAppliedLinearAccClamp=<3000 3000 3000>
    componentLinearAccScale=<0.4 0.4 0.4>
    componentLinearVelScale=<0.4 0.4 0.4>
    gravityOverride {z=-980}
    physicsBodyDefinitions [{
      boundBone {boneName=foot_ring_bk_r}
      boxExtents=<10 10 10>
      constraintSetup {angularLimitsMax=<0 0 0>, angularLimitsMin=<0 0 -70>, angularTarget=<1 0 0>}
      localJointOffset=<0 3 0>
     }]
   }
   n6 = Anim Dynamics - Bone: clavicle_ring_fr_l(ComponentPose=@n5, Alpha=1, ExternalForce="0,0,0") {
    angularDampingOverride=1
    angularSpringConstant=500
    bAngularSpring=true
    bOverrideAngularDamping=true
    bUseGravityOverride=true
    linearDampingOverride=1
    linearSpringConstant=50
    simulationSpace=RootRelative
    alphaBoolBlend {blendOption=Linear}
    boundBone {boneName=clavicle_ring_fr_l}
    componentAppliedLinearAccClamp=<3000 3000 3000>
    componentLinearAccScale=<0.4 0.4 0.4>
    componentLinearVelScale=<0.4 0.4 0.4>
    gravityOverride {z=-980}
    physicsBodyDefinitions [{
      boundBone {boneName=clavicle_ring_fr_l}
      boxExtents=<10 10 10>
      constraintSetup {angularLimitsMax=<0 0 0>, angularLimitsMin=<0 0 -70>, angularTarget=<1 0 0>}
      localJointOffset=<0 3 0>
     }]
   }
   n7 = Anim Dynamics - Bone: clavicle_ring_fr_r(ComponentPose=@n6, Alpha=1, ExternalForce="0,0,0") {
    angularDampingOverride=1
    angularSpringConstant=500
    bAngularSpring=true
    bOverrideAngularDamping=true
    bUseGravityOverride=true
    linearDampingOverride=1
    linearSpringConstant=50
    simulationSpace=RootRelative
    alphaBoolBlend {blendOption=Linear}
    boundBone {boneName=clavicle_ring_fr_r}
    componentAppliedLinearAccClamp=<3000 3000 3000>
    componentLinearAccScale=<0.4 0.4 0.4>
    componentLinearVelScale=<0.4 0.4 0.4>
    gravityOverride {z=-980}
    physicsBodyDefinitions [{
      boundBone {boneName=clavicle_ring_fr_r}
      boxExtents=<10 10 10>
      constraintSetup {angularLimitsMax=<0 0 0>, angularLimitsMin=<0 0 -70>, angularTarget=<1 0 0>}
      localJointOffset=<0 3 0>
     }]
   }
   Output Pose(Result=Component To Local(ComponentPose=@n7)) {layerGroup=None}
   n8 = "Transform (Modify) Bone - Bone: deform_glute_l"(ComponentPose=Local To Component(LocalPose=Use cached pose 'PreDyn'()), Alpha=1, Translation="0,-1,0", Rotation="0,0,0", Scale="1,1,1") {translationMode=BMM_Additive, translationSpace=BCS_BoneSpace, boneToModify {boneName=deform_glute_l}, translation {y=-1}}
   n9 = "Transform (Modify) Bone - Bone: deform_glute_r"(ComponentPose=@n8, Alpha=1, Translation="0,1,0", Rotation="0,0,0", Scale="1,1,1") {translationMode=BMM_Additive, translationSpace=BCS_BoneSpace, boneToModify {boneName=deform_glute_r}, translation {y=1}}
   n10 = Get Sim Space Settings() -> (Sim Space Settings:Anim Phys Sim Space Settings Structure)
   n11 = Anim Dynamics - Bone: deform_glute_r!(ComponentPose=@n9, Alpha=1, AngularSpringConstant:float=5000, SimSpaceSettings:Anim Phys Sim Space Settings Structure=@n10) -> (Pose) {
    angularBiasOverride=1
    angularDampingOverride=0.9999
    angularSpringConstant=5000
    bAngularSpring=true
    bGravityOverrideInSimSpace=true
    bOverrideAngularBias=true
    bOverrideAngularDamping=true
    bUseGravityOverride=true
    linearDampingOverride=0.9999
    linearSpringConstant=100
    simulationSpace=RootRelative
    boundBone {boneName=deform_glute_r}
    componentAppliedLinearAccClamp=<2 2 2>
    gravityOverride {z=-980}
    physicsBodyDefinitions [{
      boundBone {boneName=deform_glute_r}
      boxExtents=<21 21 21>
      constraintSetup {angularLimitsMax=<0 5 5>, angularLimitsMin=<0 -5 -5>, angularTarget=<1 0 0>}
      localJointOffset=<2 0 0>
     }]
    relativeSpaceBone {boneName=spine_04}
    simSpaceSettings {maxAngularAcceleration=45, maxAngularVelocity=8, simSpaceAngularAlpha=0.25}
   }
   n12 = Anim Dynamics - Bone: deform_glute_l!(ComponentPose=@n11, Alpha=1, AngularSpringConstant:float=5000, SimSpaceSettings:Anim Phys Sim Space Settings Structure=@n10) -> (Pose) {
    angularBiasOverride=1
    angularDampingOverride=0.9999
    angularSpringConstant=5000
    bAngularSpring=true
    bGravityOverrideInSimSpace=true
    bOverrideAngularBias=true
    bOverrideAngularDamping=true
    bUseGravityOverride=true
    linearDampingOverride=0.9999
    linearSpringConstant=100
    simulationSpace=RootRelative
    boundBone {boneName=deform_glute_l}
    componentAppliedLinearAccClamp=<2 2 2>
    gravityOverride {z=-980}
    physicsBodyDefinitions [{
      boundBone {boneName=deform_glute_l}
      boxExtents=<20 20 20>
      constraintSetup {angularLimitsMax=<0 5 5>, angularLimitsMin=<0 -5 -5>, angularTarget=<1 0 0>}
      localJointOffset=<-2 0 0>
     }]
    relativeSpaceBone {boneName=spine_04}
    simSpaceSettings {maxAngularAcceleration=45, maxAngularVelocity=8, simSpaceAngularAlpha=0.25}
   }
   n13 = Anim Dynamics - Bone: deform_glute_r!(ComponentPose=@n12, Alpha=1, GravityOverride:Vector="0,0,-200", LinearSpringConstant:float=180) -> (Pose) {
    angularBiasOverride=1
    angularDampingOverride=0.99
    angularSpringConstant=10000
    bGravityOverrideInSimSpace=true
    bLinearSpring=true
    bOverrideAngularBias=true
    bOverrideLinearDamping=true
    bUseGravityOverride=true
    linearDampingOverride=1
    linearSpringConstant=180
    simulationSpace=RootRelative
    boundBone {boneName=deform_glute_r}
    componentAppliedLinearAccClamp=<5 5 5>
    gravityOverride {z=-200}
    physicsBodyDefinitions [{
      boundBone {boneName=deform_glute_r}
      boxExtents=<80 80 80>
      constraintSetup {angularLimitsMax=<0 0 0>, angularLimitsMin=<0 0 0>, angularTarget=<1 0 0>, linearAxesMax=<2 2 2>, linearAxesMin=<-2 -2 -2>}
      localJointOffset=<-0.1 0.1 0>
     }]
    relativeSpaceBone {boneName=spine_04}
    simSpaceSettings {maxAngularAcceleration=50, maxAngularVelocity=10, simSpaceAngularAlpha=0.1}
   }
   n14 = Anim Dynamics - Bone: deform_glute_l!(ComponentPose=@n13, Alpha=1, GravityOverride:Vector="0,0,-200", LinearSpringConstant:float=180) -> (Pose) {
    angularBiasOverride=1
    angularDampingOverride=0.99
    angularSpringConstant=10000
    bGravityOverrideInSimSpace=true
    bLinearSpring=true
    bOverrideAngularBias=true
    bOverrideLinearDamping=true
    bUseGravityOverride=true
    linearDampingOverride=1
    linearSpringConstant=180
    simulationSpace=RootRelative
    boundBone {boneName=deform_glute_l}
    componentAppliedLinearAccClamp=<5 5 5>
    gravityOverride {z=-200}
    physicsBodyDefinitions [{
      boundBone {boneName=deform_glute_l}
      boxExtents=<90 90 90>
      constraintSetup {angularLimitsMax=<0 0 0>, angularLimitsMin=<0 0 0>, angularTarget=<1 0 0>, linearAxesMax=<2 2 2>, linearAxesMin=<-2 -2 -2>}
      localJointOffset=<0.1 -0.1 0>
     }]
    relativeSpaceBone {boneName=spine_04}
    simSpaceSettings {maxAngularAcceleration=50, maxAngularVelocity=10, simSpaceAngularAlpha=0.1}
   }
   Save cached pose 'GluteDynamics_1'(Pose=Component To Local(ComponentPose=@n14)) {cacheName=GluteDynamics_1}
   n15 = "Constraint - bag_l_base (PPPP)"!(
    ComponentPose=Local To Component(LocalPose=Use cached pose 'CtrlRig'())
    Alpha=1
    ConstraintWeights_0=0.0009
    ConstraintWeights_1:float=0.4387
    ConstraintWeights_2:float=0.5599
    ConstraintWeights_3:float=0.0005
   ) -> (Pose) {
    boneToModify {boneName=bag_l_base}
    constraintSetup [
     {targetBone {boneName=pelvis}}
     {targetBone {boneName=thigh_twist_01_l}}
     {targetBone {boneName=thigh_l}}
     {targetBone {boneName=deform_glute_l}}
    ]
    constraintWeights [1,1,1,1]
   }
   n16 = "Constraint - foot_ring_bk_base_l (PPPPPPPP)"!(
    ComponentPose=@n15
    Alpha=1
    ConstraintWeights_0=0
    ConstraintWeights_1:float=0
    ConstraintWeights_2:float=0.5374
    ConstraintWeights_3:float=0.3086
    ConstraintWeights_4:float=0.0002
    ConstraintWeights_5:float=0.0017
    ConstraintWeights_6:float=0
    ConstraintWeights_7:float=0.152
   ) -> (Pose) {
    boneToModify {boneName=foot_ring_bk_base_l}
    constraintSetup [
     {targetBone {boneName=thigh_twist_01_l}}
     {targetBone {boneName=thigh_l}}
     {targetBone {boneName=calf_twist_01_l}}
     {targetBone {boneName=foot_l}}
     {targetBone {boneName=calf_l}}
     {targetBone {boneName=calf_twist_02_l}}
     {targetBone {boneName=def_foot_corr_bk_r}}
     {targetBone {boneName=def_foot_corr_bk_l}}
    ]
    constraintWeights [1,1,1,1,1,1,1,1]
   }
   n17 = "Constraint - foot_ring_bk_base_r (PPPPPPPP)"!(
    ComponentPose=@n16
    Alpha=1
    ConstraintWeights_0=0
    ConstraintWeights_1:float=0
    ConstraintWeights_2:float=0.3086
    ConstraintWeights_3:float=0.5374
    ConstraintWeights_4:float=0.0002
    ConstraintWeights_5:float=0.0017
    ConstraintWeights_6:float=0.152
    ConstraintWeights_7:float=0
   ) -> (Pose) {
    boneToModify {boneName=foot_ring_bk_base_r}
    constraintSetup [
     {targetBone {boneName=thigh_twist_01_r}}
     {targetBone {boneName=thigh_r}}
     {targetBone {boneName=foot_r}}
     {targetBone {boneName=calf_twist_01_r}}
     {targetBone {boneName=calf_r}}
     {targetBone {boneName=calf_twist_02_r}}
     {targetBone {boneName=def_foot_corr_bk_r}}
     {targetBone {boneName=def_foot_corr_bk_l}}
    ]
    constraintWeights [1,1,1,1,1,1,1,1]
   }
   n18 = "Constraint - clavicle_ring_fr_base_l (PPPPPPPPP)"!(
    ComponentPose=@n17
    Alpha=1
    ConstraintWeights_0=0
    ConstraintWeights_1:float=0.1458
    ConstraintWeights_2:float=0.4701
    ConstraintWeights_3:float=0.0662
    ConstraintWeights_4:float=0.1219
    ConstraintWeights_5:float=0.0001
    ConstraintWeights_6:float=0.1175
    ConstraintWeights_7:float=0.0467
    ConstraintWeights_8:float=0.0317
   ) -> (Pose) {
    boneToModify {boneName=clavicle_ring_fr_base_l}
    constraintSetup [
     {targetBone {boneName=neck_01}}
     {targetBone {boneName=spine_05}}
     {targetBone {boneName=clavicle_l}}
     {targetBone {boneName=spine_04}}
     {targetBone {boneName=upperarm_twist_01_l}}
     {targetBone {boneName=deform_upperarm_bk_l}}
     {targetBone {boneName=deform_upperarm_fr_l}}
     {targetBone {boneName=deform_upperarm_up_l}}
     {targetBone {boneName=deform_pec_l}}
    ]
    constraintWeights [1,1,1,1,1,1,1,1,1]
   }
   n19 = "Constraint - clavicle_ring_fr_base_r (PPPPPPPP)"!(
    ComponentPose=@n18
    Alpha=1
    ConstraintWeights_0=0.1479
    ConstraintWeights_1:float=0.469
    ConstraintWeights_2:float=0.0663
    ConstraintWeights_3:float=0.1217
    ConstraintWeights_4:float=0
    ConstraintWeights_5:float=0.1162
    ConstraintWeights_6:float=0.0469
    ConstraintWeights_7:float=0.0319
   ) -> (Pose) {
    boneToModify {boneName=clavicle_ring_fr_base_r}
    constraintSetup [
     {targetBone {boneName=spine_05}}
     {targetBone {boneName=clavicle_r}}
     {targetBone {boneName=spine_04}}
     {targetBone {boneName=upperarm_twist_01_r}}
     {targetBone {boneName=deform_upperarm_bk_r}}
     {targetBone {boneName=deform_upperarm_fr_r}}
     {targetBone {boneName=deform_upperarm_up_r}}
     {targetBone {boneName=deform_pec_r}}
    ]
    constraintWeights [1,1,1,1,1,1,1,1]
   }
   Save cached pose 'PreDyn'(Pose=Component To Local(ComponentPose=@n19)) {cacheName=PreDyn}
   n20 = Control Rig!(
    Source=Copy Pose From Mesh() {bUseAttachedParent=true}
    Alpha
    IsParachuteOpenedAndCharUpRight:bool=Get bParachuteOpenedAndStandingUpright() -> (bParachuteOpenedAndStandingUpright:bool)
    isParachuteOpen:bool=Get bIsParachuteOpened()
    IsFrontEnd:bool=Get bIsInFrontEnd()
   ) -> (Pose) {controlRigAssetReference {blueprintRigClass=&/F_MED_BerryFlight_CtrlRig.F_MED_BerryFlight_CtrlRig_C}}
   Save cached pose 'CtrlRig'(Pose=@n20) {cacheName=CtrlRig}
  ]}
 EventGraph {nodes [Event BlueprintUpdateAnimation()]}
]
variables [Sim Space Settings {
  type=AnimPhysSimSpaceSettings
  default {maxAngularAcceleration=60, maxAngularVelocity=12, simSpaceAngularAlpha=0.7, externalAngularVelocity=<0 0 0>}
 }]
