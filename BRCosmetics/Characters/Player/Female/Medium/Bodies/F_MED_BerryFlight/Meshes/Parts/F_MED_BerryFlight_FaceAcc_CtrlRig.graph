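# Control Rig for the BerryFlight female character's face accessory, applying corrective offsets to hair and headband bones as the head moves.
# A sequence runs back-hair adjustments weighted by remapped head-to-spine_05 Euler Z, one power-curved, with some translations set by head_up.
# A spherical pose reader on the head, relative to clavicle_r and with debug drawing enabled, drives further hair offsets.
# Head delta yaw, remapped, weights a headband rotation.
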ControlRigBlueprint /BRCosmetics/Characters/Player/Female/Medium/Bodies/F_MED_BerryFlight/Meshes/Parts/F_MED_BerryFlight_FaceAcc_CtrlRig
graphs [
 New Function_ContainedGraph {nodes [
   RigVMFunctionEntryNode()
   RigUnit_GetRelativeTransformForItem = &RigUnit_GetRelativeTransformForItem(Child=(Type=Bone,Name="head"), Parent=(Type=Bone,Name="spine_05"), bChildInitial=False, bParentInitial=False)
   RigVMFunction_MathQuaternionToEuler = &RigVMFunction_MathQuaternionToEuler(RotationOrder=ZYX, Value=@RigUnit_GetRelativeTransformForItem.RelativeTransform.Rotation)
   > RigVMFunctionReturnNode(Result=@RigVMFunction_MathQuaternionToEuler.Result)
  ]}
 New Function_ContainedGraph {nodes [
   RigVMFunctionEntryNode()
   RigUnit_GetRelativeTransformForItem = &RigUnit_GetRelativeTransformForItem(Child=(Type=Bone,Name="head"), Parent=(Type=Bone,Name="spine_05"), bChildInitial=False, bParentInitial=False)
   RigVMFunction_MathQuaternionToEuler = &RigVMFunction_MathQuaternionToEuler(RotationOrder=ZYX, Value=@RigUnit_GetRelativeTransformForItem.RelativeTransform.Rotation)
   > RigVMFunctionReturnNode(Result=@RigVMFunction_MathQuaternionToEuler.Result)
  ]}
 RigVMFunctionLibrary {nodes [
   F_Head Rotation = RigVMCollapseNode()
   F_Head Rotation_1 = RigVMCollapseNode()
  ]}
 RigVMFunction_Sequence_1_ContainedGraph {nodes [
   n1 = RigVMFunctionEntryNode()
   RigVMFunction_Sequence = &RigVMFunction_Sequence(ExecuteContext=@n1.ExecuteContext)
   n2 = &RigVMFunction_Sequence(ExecuteContext=@RigVMFunction_Sequence.B)
   RigVMFunctionReturnNode(A=@RigVMFunction_Sequence.A, B=@n2.A, C=@n2.B)
  ]}
 RigVMModel {nodes [
   RigUnit_BeginExecution = &RigUnit_BeginExecution()
   RigVMFunction_Sequence = RigVMAggregateNode(ExecuteContext=@RigUnit_BeginExecution.ExecutePin)
   n1 = &RigUnit_SphericalPoseReader(
    ActiveRegionScaleFactors=(PositiveWidth=1,NegativeWidth=1,PositiveHeight=1,NegativeHeight=1)
    ActiveRegionSize=0.01
    Debug=(bDrawDebug=true,bDraw2D=False,bDrawLocalAxes=False,DebugScale=25,DebugSegments=20,DebugThickness=0.25)
    DriverAxis=(X=0,Y=1,Z=0)
    DriverItem=(Type=Bone,Name="head")
    FalloffRegionScaleFactors=(PositiveWidth=1,NegativeWidth=1,PositiveHeight=1,NegativeHeight=1)
    FalloffSize=0.1
    FlipHeightScaling=False
    FlipWidthScaling=False
    OptionalParentItem=(Type=Bone,Name="clavicle_r")
    RotationOffset=(X=0,Y=-87,Z=-5)
    ExecutePin=@RigVMFunction_Sequence.B
   )
   > &RigUnit_ModifyTransforms(
    ItemToModify="((Item=(Type=Bone,Name=\"base_hair_back_C_cr\"),Transform=(Rotation=(X=0,Y=0,Z=0,W=1),Translation=(X=0,Y=0,Z=0),Scale3D=(X=1,Y=1,Z=1))),(Item=(Type=Bone,Name=\"base_hair_back_B_lf\"),Transform=(Rotation=(X=0,Y=0,Z=0,W=1),Translation=(X=0,Y=0,Z=0),Scale3D=(X=1, ...(truncated)"
    WeightMinimum=0
    Weight=@n1.OutputParam
   )
   Get Delta Ref = RigVMFunctionReferenceNode(Item=(Type=Bone,Name="head"), ExecuteContext=@RigVMFunction_Sequence.C)
   > &RigUnit_ModifyTransforms(
    ItemToModify=((Item=(Type=Bone,Name="dyn_headband"),Transform=(Rotation=(X=0,Y=0,Z=0.0872,W=0.9962),Translation=(X=0,Y=0,Z=0),Scale3D=(X=1,Y=1,Z=1))))
    WeightMinimum=0
    Weight=&RigVMFunction_MathDoubleRemap(SourceMaximum=-40, SourceMinimum=-30, TargetMaximum=1, TargetMinimum=0, Value=@"Get Delta Ref".Result.Rotation.Yaw).Result
   )
   F_Head Rotation = RigVMFunctionReferenceNode()
   Remap = &RigVMFunction_MathDoubleRemap(SourceMaximum=-80, SourceMinimum=0, TargetMaximum=1, TargetMinimum=0, Value=@"F_Head Rotation".Result.Z)
   n2 = RigVMVariableNode() {variable=head_up}
   &RigUnit_ModifyTransforms(
    ItemToModify="((Item=(Type=Bone,Name=\"dyn_hair_back_01_lf\"),Transform=(Rotation=(X=0,Y=0,Z=-0.1736,W=0.9848),Translation=(X=-5,Y=0,Z=0),Scale3D=(X=1,Y=1,Z=1))),(Item=(Type=Bone,Name=\"base_hair_back_rt\"),Transform=(Rotation=(X=0,Y=0,Z=0,W=1),Translation=(X=-5,Y=0,Z=0),Sc ...(truncated)"
    WeightMinimum=0
    Weight=@Remap.Result
    ItemToModify.0.Transform.Translation.X=@n2.Value
    ItemToModify.1.Transform.Translation.X=@n2.Value
    ItemToModify.2.Transform.Translation.X=@n2.Value
    ExecutePin=@RigVMFunction_Sequence.A
   )
   > &RigUnit_ModifyTransforms(
    ItemToModify="((Item=(Type=Bone,Name=\"dyn_hair_back_01_cr\"),Transform=(Rotation=(X=0,Y=0,Z=-0.1736,W=0.9848),Translation=(X=0,Y=0,Z=0),Scale3D=(X=1,Y=1,Z=1))),(Item=(Type=Bone,Name=\"base_hair_back_rt\"),Transform=(Rotation=(X=0,Y=0,Z=-0.1736,W=0.9848),Translation=(X=0,Y= ...(truncated)"
    WeightMinimum=0
    Weight=@Remap.Result
   )
   F_Head Rotation_4 = RigVMFunctionReferenceNode()
   Power = /Script/RigVM.RigVMFunction_MathDoublePow(
    B=0.7
    A=&RigVMFunction_MathDoubleRemap(SourceMaximum=-55, SourceMinimum=0, TargetMaximum=1, TargetMinimum=0, Value=@"F_Head Rotation_4".Result.Z).Result
   ) {is=struct RigVMUnitNode, template="Power::Execute(in A,in B,out Result)"}
   > &RigUnit_ModifyTransforms(
    ItemToModify="((Item=(Type=Bone,Name=\"base_hair_back_rt\"),Transform=(Rotation=(X=0,Y=0,Z=0,W=1),Translation=(X=0,Y=0,Z=0),Scale3D=(X=1,Y=1,Z=1))),(Item=(Type=Bone,Name=\"base_hair_back_cr\"),Transform=(Rotation=(X=0,Y=0,Z=0,W=1),Translation=(X=0,Y=0,Z=0),Scale3D=(X=1,Y=1, ...(truncated)"
    WeightMinimum=0
    Weight=@Power.Result
   )
  ]}
]
hierarchy [
 root {loc=<0 0 0>, rot=<0 0 0 1>}
 pelvis {parent=root, loc=<0 -0.5301 98.0571>, rot=<0.0127 0.707 -0.0127 -0.707>}
 spine_01 {parent=pelvis, loc=<6.1208 -0.9328 0>, rot=<0 0 0.034 -0.9994>}
 spine_02 {parent=spine_01, loc=<6.5996 -0.2095 0>, rot=<0 0 0.0238 -0.9997>}
 spine_03 {parent=spine_02, loc=<6.4914 -0.1237 0>, rot=<0 0 0.0149 -0.9999>}
 spine_04 {parent=spine_03, loc=<6.9598 0.0794 0>, rot=<0 0 -0.0132 -0.9999>}
 spine_05 {parent=spine_04, loc=<6.9987 0 0>, rot=<0 0 -0.0015 -1>}
 neck_01 {parent=spine_05, loc=<11.3138 0.1529 0>, rot=<0 0 -0.2003 -0.9797>}
 neck_02 {parent=neck_01, loc=<4.5627 0 0>, rot=<0 0 0.0504 -0.9987>}
 head {parent=neck_02, loc=<5.3914 0 0>, rot=<0 0 0.115 -0.9934>}
 dyn_headband {parent=head, loc=<2.5348 1.2235 0>, rot=<0 0 -0.0041 -1>}
 base_hair_front_B_rt {parent=head, loc=<-0.427 -2.1484 6.4019>, rot=<-0.1211 -0.8369 -0.5147 -0.1416>}
 dyn_hair_front_B_01_rt {parent=base_hair_front_B_rt, loc=<0 0 0>, rot=<0 0 0 1>}
 base_hair_front_A_rt {parent=head, loc=<10.3678 6.3694 4.8392>, rot=<-0.093 -0.8054 -0.5072 0.2923>}
 dyn_hair_front_A_01_rt {parent=base_hair_front_A_rt, loc=<0 0 0>, rot=<0 0 0 1>}
 dyn_hair_front_A_02_rt {parent=dyn_hair_front_A_01_rt, loc=<3.9462 0 0>, rot=<0 0 0 1>}
 base_hair_front_B_lf {parent=head, loc=<-0.427 -2.1484 -6.4019>, rot=<0.6161 0.347 0.619 0.3419>}
 dyn_hair_front_B_01_lf {parent=base_hair_front_B_lf, loc=<0 0 0>, rot=<0 0 0 1>}
 base_hair_front_A_lf {parent=head, loc=<6.0321 4.1815 -7.9695>, rot=<0.5893 0.3909 0.5925 0.386>}
 dyn_hair_front_A_01_lf {parent=base_hair_front_A_lf, loc=<0 0 0>, rot=<0 0 0 1>}
 base_hair_back_rt {parent=head, loc=<11.7696 -5.4045 5.1862>, rot=<0.0375 -0.9993 0 0>}
 dyn_hair_back_01_rt {parent=base_hair_back_rt, loc=<0 0 0>, rot=<0 0 0 1>}
 dyn_hair_back_02_rt {parent=dyn_hair_back_01_rt, loc=<6.829 0 0>, rot=<0 0 0 1>}
 dyn_hair_back_03_rt {parent=dyn_hair_back_02_rt, loc=<4.295 0 0>, rot=<0 0 0 1>}
 base_hair_back_C_cr {parent=dyn_hair_back_03_rt, loc=<-1.5446 -1.8272 -0.9162>, rot=<0.6067 -0.3632 -0.5811 0.4029>}
 dyn_hair_back_C_01_cr {parent=base_hair_back_C_cr, loc=<0 0 0>, rot=<0 0 0 1>}
 base_hair_side_rt {parent=dyn_hair_back_03_rt, loc=<-0.6102 0.3596 -2.5465>, rot=<0.6929 -0.141 -0.6819 0.187>}
 dyn_hair_side_rt {parent=base_hair_side_rt, loc=<0 0 0>, rot=<0 0 0 1>}
 base_hair_back_cr {parent=head, loc=<11.7696 -5.4045 0>, rot=<0.0375 -0.9993 0 0>}
 dyn_hair_back_01_cr {parent=base_hair_back_cr, loc=<0 0 0>, rot=<0 0 0 1>}
 dyn_hair_back_02_cr {parent=dyn_hair_back_01_cr, loc=<6.829 0 0>, rot=<0 0 0 1>}
 dyn_hair_back_03_cr {parent=dyn_hair_back_02_cr, loc=<4.295 0 0>, rot=<0 0 0 1>}
 base_hair_back_B_lf {parent=dyn_hair_back_03_cr, loc=<-0.4478 -1.9465 -0.8118>, rot=<-0.5164 0.483 0.483 -0.5164>}
 dyn_hair_back_B_01_lf {parent=base_hair_back_B_lf, loc=<0 0 0>, rot=<0 0 0 1>}
 base_hair_back_lf {parent=head, loc=<11.7696 -5.4045 -5.186>, rot=<0.0375 -0.9993 0 0>}
 dyn_hair_back_01_lf {parent=base_hair_back_lf, loc=<0 0 0>, rot=<0 0 0 1>}
 dyn_hair_back_02_lf {parent=dyn_hair_back_01_lf, loc=<6.829 0 0>, rot=<0 0 0 1>}
 dyn_hair_back_03_lf {parent=dyn_hair_back_02_lf, loc=<4.295 0 0>, rot=<0 0 0 1>}
 base_hair_back_A_lf {parent=dyn_hair_back_03_lf, loc=<-0.5951 -2.3785 0.2829>, rot=<-0.3492 0.6149 0.3074 -0.6368>}
 dyn_hair_back_A_lf {parent=base_hair_back_A_lf, loc=<0 0 0>, rot=<0 0 0 1>}
 base_hair_side_lf {parent=dyn_hair_back_03_lf, loc=<0.7457 1.7847 2.0785>, rot=<-0.1988 0.6786 0.1531 -0.6903>}
 dyn_hair_side_lf {parent=base_hair_side_lf, loc=<0 0 0>, rot=<0 0 0 1>}
 base_hair_back_D_lf {parent=dyn_hair_back_03_lf, loc=<-0.3747 0.5204 1.86>, rot=<-0.2136 0.6741 0.1681 -0.6868>}
 dyn_hair_back_D_lf {parent=base_hair_back_D_lf, loc=<0 0 0>, rot=<0 0 0 1>}
 base_hair_back_C_lf {parent=dyn_hair_back_03_lf, loc=<-1.2473 -2.6405 -2.6472>, rot=<-0.5071 0.4928 0.4731 -0.5255>}
 dyn_hair_back_C_lf {parent=base_hair_back_C_lf, loc=<0 0 0>, rot=<0 0 0 1>}
 base_hair_front_side_rt {parent=head, loc=<-0.2005 -4.5149 7.2645>, rot=<0.6745 -0.2122 0.6727 -0.2177>}
 dyn_hair_front_side_rt {parent=base_hair_front_side_rt, loc=<0 0 0>, rot=<0 0 0 1>}
 base_hair_Forehead_l {parent=head, loc=<9.6366 7.653 -6.5212>, rot=<0.1317 -0.715 -0.6861 0.0262>}
 dyn_hair_Forehead_l_01 {parent=base_hair_Forehead_l, loc=<0 0 0>, rot=<0 0 0 1>}
 dyn_hair_Forehead_l_04 {parent=dyn_hair_Forehead_l_01, loc=<5.7149 -1.9275 -0.4336>, rot=<-0.3813 0.025 0.3616 -0.8504>}
 dyn_hair_Forehead_l_02 {parent=dyn_hair_Forehead_l_01, loc=<4 0 0>, rot=<-0.0152 -0.0915 0.1521 -0.984>}
 dyn_hair_Forehead_l_03 {parent=dyn_hair_Forehead_l_02, loc=<4 0 0>, rot=<0 0 0 1>}
 base_hair_Forehead_r {parent=head, loc=<8.4289 9.4517 3.9114>, rot=<0.114 0.3328 0.9136 -0.204>}
 dyn_hair_Forehead_r_01 {parent=base_hair_Forehead_r, loc=<0 0 0>, rot=<0 0 0 1>}
 dyn_hair_Forehead_r_02 {parent=dyn_hair_Forehead_r_01, loc=<4 0 0>, rot=<0 0 -0.328 -0.9447>}
 dyn_hair_Forehead_r_03 {parent=dyn_hair_Forehead_r_02, loc=<3 0 0>, rot=<0 0 0 1>}
]
variables [head_up {type=double, default=-6}]
